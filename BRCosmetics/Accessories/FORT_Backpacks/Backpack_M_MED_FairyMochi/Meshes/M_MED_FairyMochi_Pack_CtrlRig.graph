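# Control Rig for the FairyMochi backpack cosmetic, applying a small fixed pose adjustment to the pack's skeleton.
# From begin execution, two chained Modify Transforms units run at full weight: a slight translation on base_COG, then a slight rotation on the right leg bone.
# The skeleton is a small figure on a chair bone, with dyn_-prefixed spine, head, arm and leg bones.
# The graph contains no simulation, variables or control logic.
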ControlRigBlueprint /BRCosmetics/Accessories/FORT_Backpacks/Backpack_M_MED_FairyMochi/Meshes/M_MED_FairyMochi_Pack_CtrlRig
graphs [
 RigVMFunctionLibrary {links [], nodes []}
 RigVMModel {nodes [
   RigUnit_BeginExecution = &RigUnit_BeginExecution()
   &RigUnit_ModifyTransforms(
    ItemToModify=((Item=(Type=Bone,Name="base_COG"),Transform=(Rotation=(X=0,Y=0,Z=0,W=1),Translation=(X=0,Y=0.5,Z=0),Scale3D=(X=1,Y=1,Z=1))))
    Weight=1
    WeightMinimum=0
    ExecutePin=@RigUnit_BeginExecution.ExecutePin
   )
   > &RigUnit_ModifyTransforms(
    ItemToModify=((Item=(Type=Bone,Name="dyn_leg_r"),Transform=(Rotation=(X=0.0005,Y=-0.0174,Z=0.0262,W=0.9995),Translation=(X=0,Y=0,Z=0),Scale3D=(X=1,Y=1,Z=1))))
    Weight=1
    WeightMinimum=0
   )
  ]}
]
hierarchy [
 root {loc=<0 0 0>, rot=<0 0 0 1>}
 pack {parent=root, loc=<0 -0.7316 0>, rot=<0 0 0 1>}
 base_chair {parent=pack, loc=<0.0564 -9.0165 -13.4327>, rot=<0.7071 0 0.7071 0>}
 base_COG {parent=base_chair, loc=<2.9162 0 0>, rot=<0 0 0 1>}
 dyn_spine {parent=base_COG, loc=<3.6626 0 0>, rot=<-0.114 -0.0014 0.0118 -0.9934>}
 dyn_head {parent=dyn_spine, loc=<9.9643 0 0>, rot=<0 0 0 1>}
 dyn_shoulder_r {parent=dyn_spine, loc=<6.8112 0.9654 5.0287>, rot=<-0.4457 -0.7313 -0.2922 0.4257>}
 dyn_elbow_r {parent=dyn_shoulder_r, loc=<3.651 0 0>, rot=<0 0.0221 0 -0.9998>}
 dyn_wrist_r {parent=dyn_elbow_r, loc=<2.279 0 0>, rot=<-0.2307 -0.2144 -0.2846 -0.9054>}
 dyn_shoulder_l {parent=dyn_spine, loc=<7.1671 -0.0893 -5.3194>, rot=<0.307 -0.627 -0.5589 -0.4475>}
 dyn_elbow_l {parent=dyn_shoulder_l, loc=<3.6509 0 0>, rot=<0 0.0286 0 -0.9996>}
 dyn_wrist_l {parent=dyn_elbow_l, loc=<2.279 0 0>, rot=<0.4868 -0.1476 0.2315 -0.8292>}
 dyn_leg_r {parent=base_COG, loc=<2.5475 1.945 4.3833>, rot=<0.2612 -0.2673 -0.6823 -0.6283>}
 dyn_Knee_r {parent=dyn_leg_r, loc=<4.943 0 0>, rot=<0 0 -0.0044 -1>}
 dyn_ankle_r {parent=dyn_Knee_r, loc=<4.2511 0 0>, rot=<0 0 0 1>}
 dyn_leg_L {parent=base_COG, loc=<1.366 1.8896 -4.2326>, rot=<-0.5791 -0.7558 -0.2945 0.082>}
 dyn_Knee_l {parent=dyn_leg_L, loc=<4.9137 0 0>, rot=<0 0 0.1493 -0.9888>}
 dyn_ankle_l {parent=dyn_Knee_l, loc=<4.2299 0 0>, rot=<0 0 0 1>}
]
